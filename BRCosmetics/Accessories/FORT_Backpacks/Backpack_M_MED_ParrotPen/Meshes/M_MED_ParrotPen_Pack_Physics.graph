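# Physics asset for the ParrotPen backpack cosmetic, giving capsule bodies and joint limits to its dynamic cape bones.
# Five chains hang from a tiny base_cape anchor: a three-bone center chain, mirrored two-bone A chains, and single-bone B bodies on each side.
# Constraints allow more swing on one axis than the other, with twist locked on some joints.
# A large pack capsule has no constraints, likely serving as a collision volume for the cape.
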PhysicsAsset /BRCosmetics/Accessories/FORT_Backpacks/Backpack_M_MED_ParrotPen/Meshes/M_MED_ParrotPen_Pack_Physics
bodies {
 base_cape [{shapeType=Capsule, length=0.1, radius=0.1, shapeName=base_cape_capsule, rotation=<rot 90 0 0>}]
 dyn_capeA_02_l [{shapeType=Capsule, length=5, radius=1.3, shapeName=dyn_capeA_02_r_capsule, center=<3.5 0 0>, rotation=<rot 90 0 0>}]
 dyn_capeA_02_r [{shapeType=Capsule, length=5, radius=1.3, shapeName=dyn_capeA_02_r_capsule, center=<3.5 0 0>, rotation=<rot 90 0 0>}]
 dyn_capeA_03_l [{shapeType=Capsule, length=3.5, radius=1.3, shapeName=dyn_capeA_03_l_capsule, center=<3 0 0>, rotation=<rot 90 0 0>}]
 dyn_capeA_03_r [{shapeType=Capsule, length=3.5, radius=1.3, shapeName=dyn_capeA_03_r_capsule, center=<3 0 0>, rotation=<rot 90 0 0>}]
 dyn_capeB_02_l [{shapeType=Capsule, length=2, radius=2, shapeName=dyn_capeB_02_r_capsule, center=<1 0 0>, rotation=<rot 90 0 0>}]
 dyn_capeB_02_r [{shapeType=Capsule, length=2, radius=2, shapeName=dyn_capeB_02_r_capsule, center=<1 0 0>, rotation=<rot 90 0 0>}]
 dyn_cape_02 [{shapeType=Capsule, length=5, radius=1.3, shapeName=dyn_cape_02_capsule, center=<3.5 0 0>, rotation=<rot 90 0 0>}]
 dyn_cape_03 [{shapeType=Capsule, length=4, radius=1.3, shapeName=dyn_cape_03_capsule, center=<3.5 0 0>, rotation=<rot 90 0 0>}]
 dyn_cape_04 [{shapeType=Capsule, length=2.5, radius=1, shapeName=dyn_cape_04_capsule, center=<1.6032 0 0>, rotation=<rot 90 0 0>}]
 pack [{shapeType=Capsule, length=11, radius=9, shapeName=pack_capsule, center=<0 6.4591 -20.0064>}]
}
constraints [
 {bone1Name=dyn_cape_03, bone2Name=dyn_cape_02, swing1LimitDegrees=25, swing2LimitDegrees=20, twistLimitDegrees=1, twistMotion=Locked}
 {bone1Name=dyn_cape_04, bone2Name=dyn_cape_03, swing1LimitDegrees=25, swing2LimitDegrees=15, twistLimitDegrees=1}
 {bone1Name=dyn_capeA_03_r, bone2Name=dyn_capeA_02_r, swing1LimitDegrees=25, swing2LimitDegrees=5, twistLimitDegrees=1}
 {bone1Name=dyn_capeA_03_l, bone2Name=dyn_capeA_02_l, swing1LimitDegrees=25, swing2LimitDegrees=5, twistLimitDegrees=1}
 {bone1Name=dyn_capeB_02_r, bone2Name=base_cape, swing1LimitDegrees=25, swing2LimitDegrees=2, twistLimitDegrees=1, twistMotion=Locked}
 {bone1Name=dyn_capeA_02_r, bone2Name=base_cape, swing1LimitDegrees=25, swing2LimitDegrees=10, twistLimitDegrees=1}
 {bone1Name=dyn_cape_02, bone2Name=base_cape, swing1LimitDegrees=25, swing2LimitDegrees=5, twistLimitDegrees=1}
 {bone1Name=dyn_capeA_02_l, bone2Name=base_cape, swing1LimitDegrees=25, swing2LimitDegrees=10, twistLimitDegrees=1}
 {bone1Name=dyn_capeB_02_l, bone2Name=base_cape, swing1LimitDegrees=25, swing2LimitDegrees=2, twistLimitDegrees=1, twistMotion=Locked}
]
