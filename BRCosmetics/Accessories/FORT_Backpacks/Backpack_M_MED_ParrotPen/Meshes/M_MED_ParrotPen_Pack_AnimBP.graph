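# Animation Blueprint for the ParrotPen backpack cosmetic, posing its pack and cape bones with procedural offsets and physics.
# Each update sets alphas from character state, maps spine bend to pack offset, and updates gravity and floating-dependent sim-space settings.
# Additive bone modifiers feed a RigidBody simulation in root bone space, followed by more pack and cape adjustments.
# Alpha_isSprinting actually follows jumping or falling, despite its name.
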AnimBlueprint /BRCosmetics/Accessories/FORT_Backpacks/Backpack_M_MED_ParrotPen/Meshes/M_MED_ParrotPen_Pack_AnimBP
target_skeleton=&/M_MED_ParrotPen_Pack_Skeleton
graphs [
 AnimGraph {nodes [
   n1 = "Transform (Modify) Bone - Bone: pack"(
    ComponentPose=Local To Component(LocalPose=Input Pose())
    Alpha=Get Alpha_isNotFrontEnd() -> (Alpha_isNotFrontEnd:double)
    Translation="0,1,0"
    Rotation="0,0,0"
   ) {
    rotationMode=BMM_Additive
    rotationSpace=BCS_BoneSpace
    translationMode=BMM_Additive
    translationSpace=BCS_BoneSpace
    alphaBoolBlend {blendOption=Linear}
    alphaScaleBiasClamp {bInterpResult=true}
    boneToModify {boneName=pack}
    translation {y=1}
   }
   n2 = "Transform (Modify) Bone - Bone: pack"(
    ComponentPose=@n1
    Alpha=Get Alpha_isNotJumping() -> (Alpha_isNotJumping:double)
    Translation=Get backpack_translation()
    Rotation=Get backpack_rotate()
   ) {
    rotationMode=BMM_Additive
    rotationSpace=BCS_BoneSpace
    translationMode=BMM_Additive
    translationSpace=BCS_BoneSpace
    alphaBoolBlend {blendOption=Linear}
    alphaScaleBiasClamp {bInterpResult=true}
    boneToModify {boneName=pack}
    translation {y=2}
   }
   n3 = "Transform (Modify) Bone - Bone: pack"(ComponentPose=@n2, Alpha=Get Alpha_Skydive() -> (Alpha_Skydive:double), Translation="0,-0.3,0", Rotation="0,0,0") {
    rotationMode=BMM_Additive
    rotationSpace=BCS_BoneSpace
    translationMode=BMM_Additive
    translationSpace=BCS_BoneSpace
    alphaBoolBlend {blendOption=Linear}
    alphaScaleBiasClamp {bInterpResult=true, interpSpeedDecreasing=25, interpSpeedIncreasing=25}
    boneToModify {boneName=pack}
    translation {y=-0.3}
   }
   n4 = "Transform (Modify) Bone - Bone: pack"(ComponentPose=@n3, Alpha=Get Alpha_SkydiveMode() -> (Alpha_SkydiveMode:double), Translation="0,1,-3", Rotation="0,0,0") {
    rotationMode=BMM_Additive
    rotationSpace=BCS_BoneSpace
    translationMode=BMM_Additive
    translationSpace=BCS_BoneSpace
    alphaBoolBlend {blendOption=Linear}
    alphaScaleBiasClamp {bInterpResult=true}
    boneToModify {boneName=pack}
    translation {y=1, z=-3}
   }
   n5 = "Transform (Modify) Bone - Bone: pack"(ComponentPose=@n4, Alpha=Get Alpha_isSprinting() -> (Alpha_isSprinting:double), Translation="0,0,0", Rotation="0,0,0") {
    rotationMode=BMM_Additive
    rotationSpace=BCS_BoneSpace
    translationMode=BMM_Additive
    translationSpace=BCS_BoneSpace
    alphaBoolBlend {blendOption=Linear}
    alphaScaleBiasClamp {bInterpResult=true, interpSpeedDecreasing=25, interpSpeedIncreasing=25}
    boneToModify {boneName=pack}
   }
   n6 = "Transform (Modify) Bone - Bone: pack"(ComponentPose=@n5, Alpha=Get Alpha_SkyDiveFloating() -> (Alpha_SkyDiveFloating:float), Translation="0,5,0", Rotation="0,0,0") {
    rotationMode=BMM_Additive
    rotationSpace=BCS_BoneSpace
    translationMode=BMM_Additive
    translationSpace=BCS_BoneSpace
    alphaBoolBlend {blendOption=Linear}
    alphaScaleBiasClamp {bInterpResult=true, interpSpeedDecreasing=25, interpSpeedIncreasing=25}
    boneToModify {boneName=pack}
    translation {y=5}
   }
   n7 = "Transform (Modify) Bone - Bone: pack"(ComponentPose=@n6, Alpha=Get Alpha_Crouching() -> (Alpha_Crouching:float), Translation="0,-0.5,0", Rotation="0,0,10") {
    rotationMode=BMM_Additive
    rotationSpace=BCS_BoneSpace
    translationMode=BMM_Additive
    translationSpace=BCS_BoneSpace
    alphaBoolBlend {blendOption=Linear}
    alphaScaleBiasClamp {bInterpResult=true, interpSpeedDecreasing=25, interpSpeedIncreasing=25}
    boneToModify {boneName=pack}
    rotation {roll=10}
    translation {y=-0.5}
   }
   n8 = RigidBody!(
    ComponentPose=@n7
    Alpha
    OverrideWorldGravity:Vector=Get GravityOverride()
    ExternalForce
    SimSpaceSettings:Sim Space Settings Structure=Get Sim Space Settings() -> (Sim Space Settings:Sim Space Settings Structure)
   ) -> (Pose) {
    bOverrideWorldGravity=true
    overridePhysicsAsset=&/M_MED_ParrotPen_Pack_Physics
    simulationSpace=BaseBoneSpace
    alphaBoolBlend {blendOption=Linear}
    baseBoneRef {boneName=root}
    componentAppliedLinearAccClamp=<3000 3000 3000>
    componentLinearAccScale=<0.3 0.2 0.2>
    componentLinearVelScale=<0.3 0.2 0.2>
    simSpaceSettings {
     maxAngularAcceleration=1000
     maxAngularVelocity=3
     maxLinearAcceleration=700
     maxLinearVelocity=500
     velocityScaleZ=0
     worldAlpha=1
     externalLinearDragV=<0.2 0.2 0.2>
    }
   }
   n9 = "Transform (Modify) Bone - Bone: pack"(ComponentPose=@n8, Alpha=Get Alpha_isEmote() -> (Alpha_isEmote:double), Translation="0,0,0", Rotation="0,0,8") {
    rotationMode=BMM_Additive
    rotationSpace=BCS_BoneSpace
    translationMode=BMM_Additive
    translationSpace=BCS_BoneSpace
    alphaBoolBlend {blendOption=Linear}
    alphaScaleBiasClamp {bInterpResult=true}
    boneToModify {boneName=pack}
    rotation {roll=8}
   }
   n10 = Get Alpha_isFrontEnd() -> (Alpha_isFrontEnd:double)
   n11 = "Transform (Modify) Bone - Bone: base_cape"(ComponentPose=@n9, Alpha=@n10, Translation="0,0,0", Rotation="-3,0,0", Scale="1,1,1") {
    rotationMode=BMM_Additive
    rotationSpace=BCS_BoneSpace
    translationMode=BMM_Additive
    translationSpace=BCS_BoneSpace
    alphaBoolBlend {blendOption=Linear}
    alphaScaleBiasClamp {bInterpResult=true}
    boneToModify {boneName=base_cape}
    rotation {pitch=-3}
   }
   n12 = "Transform (Modify) Bone - Bone: dyn_cape_03"(ComponentPose=@n11, Alpha=@n10, Translation="0,0,0", Rotation="5,0,0", Scale="1,1,1") {
    rotationMode=BMM_Additive
    rotationSpace=BCS_BoneSpace
    translationMode=BMM_Additive
    translationSpace=BCS_BoneSpace
    alphaBoolBlend {blendOption=Linear}
    alphaScaleBiasClamp {bInterpResult=true}
    boneToModify {boneName=dyn_cape_03}
    rotation {pitch=5}
   }
   n13 = "Transform (Modify) Bone - Bone: base_cape"(ComponentPose=@n12, Alpha=Get Alpha_Skydive(), Translation="0,0,0", Rotation="-10,0,0", Scale="1,1,1") {
    rotationMode=BMM_Additive
    rotationSpace=BCS_BoneSpace
    translationMode=BMM_Additive
    translationSpace=BCS_BoneSpace
    alphaBoolBlend {blendOption=Linear}
    alphaScaleBiasClamp {bInterpResult=true, interpSpeedDecreasing=2, interpSpeedIncreasing=2}
    boneToModify {boneName=base_cape}
    rotation {pitch=-10}
   }
   n14 = "Transform (Modify) Bone - Bone: pack"(ComponentPose=@n13, Alpha=Get Alpha_SkyDiveFloating(), Translation="0,-5.7,0", Rotation="0,0,0") {
    rotationMode=BMM_Additive
    rotationSpace=BCS_BoneSpace
    translationMode=BMM_Additive
    translationSpace=BCS_BoneSpace
    alphaBoolBlend {blendOption=Linear}
    alphaScaleBiasClamp {bInterpResult=true, interpSpeedDecreasing=2, interpSpeedIncreasing=2}
    boneToModify {boneName=pack}
    translation {y=-5.7}
   }
   n15 = "Transform (Modify) Bone - Bone: dyn_cape_03"(ComponentPose=@n14, Alpha=Get Alpha_isSprinting(), Translation="0,0,0", Rotation="30,0,0", Scale="1,1,1") {
    rotationMode=BMM_Additive
    rotationSpace=BCS_BoneSpace
    translationMode=BMM_Additive
    translationSpace=BCS_BoneSpace
    alphaBoolBlend {blendOption=Linear}
    alphaScaleBiasClamp {bInterpResult=true}
    boneToModify {boneName=dyn_cape_03}
    rotation {pitch=30}
   }
   Output Pose(Result=Component To Local(ComponentPose=@n15)) {layerGroup=None}
  ]}
 EventGraph {nodes [
   Event BlueprintUpdateAnimation()
   > n1 = Sequence!(execute) -> (
    then_0
    then_1
    then_2:Exec
    then_3:Exec
    then_4:Exec
    then_5:Exec
    then_6:Exec
    then_7:Exec
    then_8:Exec
    then_9:Exec
    then_10:Exec
    then_11:Exec
   )
   n2 = Get bIsSkydiveFloating()
   n3 = Branch(execute=@n1.then_1, Condition=@n2)
   > Set Sim Space Settings(Sim Space Settings:Sim Space Settings Structure=Get Sim Space Settings_isFloatingTrue() -> (Sim Space Settings_isFloatingTrue:Sim Space Settings Structure)) -> (Output_Get:Sim Space Settings Structure)
   Set Sim Space Settings(
    execute=@n3.else
    Sim Space Settings=Get Sim Space Settings_isFloatingFalse() -> (Sim Space Settings_isFloatingFalse:Sim Space Settings Structure)
   )
   n4 = Branch(execute=@n1.then_10, Condition=NOT Boolean(A=Get bIsInFrontEnd()))
   > Set Alpha_isNotFrontEnd(Alpha_isNotFrontEnd:double=1) -> (Output_Get:double)
   Set Alpha_isNotFrontEnd(execute=@n4.else, Alpha_isNotFrontEnd=0)
   n5 = Branch(execute=@n1.then_11, Condition=Get bIsPlayingEmote())
   > Set Alpha_isEmote(Alpha_isEmote:double=1) -> (Output_Get:double)
   Set Alpha_isEmote(execute=@n5.else, Alpha_isEmote=0)
   n6 = Branch(execute=@n1.then_2, Condition=@n2)
   > Set Alpha_SkyDiveFloating(Alpha_SkyDiveFloating:float=1) -> (Output_Get:float)
   Set Alpha_SkyDiveFloating(execute=@n6.else, Alpha_SkyDiveFloating=0)
   n7 = Branch(execute=@n1.then_3, Condition=Get bIsCrouching())
   > Set Alpha_Crouching(Alpha_Crouching:float=1) -> (Output_Get:float)
   Set Alpha_Crouching(execute=@n7.else, Alpha_Crouching=0)
   n8 = Branch(execute=@n1.then_4, Condition=Get bIsSkydiveDiveMode())
   > Set Alpha_SkydiveMode(Alpha_SkydiveMode:double=1) -> (Output_Get:double)
   Set Alpha_SkydiveMode(execute=@n8.else, Alpha_SkydiveMode=0)
   n9 = Branch(execute=@n1.then_5, Condition=Get bIsSkyDiving())
   > Set Alpha_Skydive(Alpha_Skydive:double=1) -> (Output_Get:double)
   Set Alpha_Skydive(execute=@n9.else, Alpha_Skydive=0)
   n10 = Branch(execute=@n1.then_6, Condition=OR Boolean(A=Get bIsJumping(), B=Get bIsFalling()))
   > Set Alpha_isSprinting(Alpha_isSprinting:double=1) -> (Output_Get:double)
   Set Alpha_isSprinting(execute=@n10.else, Alpha_isSprinting=0)
   n11 = Branch(execute=@n1.then_7, Condition=NOR Boolean!(A:bool=Get bIsFalling(), B:bool=Get bIsJumping()) -> (ReturnValue:bool))
   > Set Alpha_isNotJumping(Alpha_isNotJumping:double=1) -> (Output_Get:double)
   Set Alpha_isNotJumping(execute=@n11.else, Alpha_isNotJumping=0)
   n12 = Gravity Override State Function DEPRECATED(
    execute=@n1.then_8
    CharacterPartAnimInstance=Self-Reference()
    GameTimeInSeconds=GetGameTimeInSeconds()
    GravityOverrideParams=Get Gravity Override Params()
   )
   > Set GravityOverride(GravityOverride=@n12)
   n13 = Branch(execute=@n1.then_9, Condition=Get bIsInFrontEnd())
   > Set Alpha_isFrontEnd(Alpha_isFrontEnd:double=1) -> (Output_Get:double)
   Set Alpha_isFrontEnd(execute=@n13.else, Alpha_isFrontEnd=0)
   n14 = GetOwnerSkeletalMeshComponent()
   n15 = BreakRotator(InRot=BreakTransform(InTransform=GetDeltaTransformFromRefPose(self=@n14, BoneName=spine_05, BaseName=spine_01)).Rotation)
   Set backpack_rotate(execute=@n1.then_0, backpack_rotate=MakeRotator(Roll=MapRangeClamped(Value=@n15.Yaw, InRangeA=0, InRangeB=-30, OutRangeB=35)))
   > Set backpack_translation(backpack_translation=MakeVector(Y=MapRangeClamped(Value=@n15.Yaw, InRangeA=0, InRangeB=-30, OutRangeA=1, OutRangeB=-5)))
   TryGetPawnOwner()
   Get bIsPlayingEmote()
  ]}
]
variables [
 Alpha_Crouching {type=number, default=1}
 Alpha_isEmote {type=number}
 Alpha_isFrontEnd {type=number}
 Alpha_isNotFrontEnd {type=number}
 Alpha_isNotJumping {type=number}
 Alpha_isSprinting {type=number}
 Alpha_Skydive {type=number}
 Alpha_SkyDiveFloating {type=number, default=1}
 Alpha_SkydiveMode {type=number}
 backpack_rotate {type=Rotator}
 backpack_translation {type=Vector}
 Gravity Override Params {type=GravityOverrideParamsStruct}
 GravityOverride {type=Vector}
 Sim Space Settings {
  type=SimSpaceSettings
  default {
   dampingAlpha=1
   maxAngularAcceleration=1000
   maxAngularVelocity=3
   maxLinearAcceleration=700
   maxLinearVelocity=500
   velocityScaleZ=0
   worldAlpha=1
   externalLinearDragV=<0.2 0.2 0.2>
  }
 }
 Sim Space Settings_isFloatingFalse {
  type=SimSpaceSettings
  default {
   dampingAlpha=1
   maxAngularAcceleration=1000
   maxAngularVelocity=3
   maxLinearAcceleration=700
   maxLinearVelocity=500
   velocityScaleZ=0
   worldAlpha=0.5
   externalLinearDragV=<0.2 0.2 0>
  }
 }
 Sim Space Settings_isFloatingTrue {
  type=SimSpaceSettings
  default {
   dampingAlpha=1
   maxAngularAcceleration=600
   maxAngularVelocity=3
   maxLinearAcceleration=500
   maxLinearVelocity=300
   velocityScaleZ=0
   worldAlpha=0.25
   externalLinearDragV=<0.2 0.2 0>
  }
 }
]
